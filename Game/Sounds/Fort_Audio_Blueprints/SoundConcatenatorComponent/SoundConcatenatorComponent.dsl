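; Sound concatenator component that plays an array of sounds one after another, with an optional delay following each.
; Play runs only where cosmetic events can execute, stores the attach component, then spawns the current item attached to it, or as 2D without one.
; Each sound's audio-finished event triggers the item's delay, on a timer if nonzero, then increments the index and plays the next.
; The chain stops when the current item's sound is invalid.
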(fn AdvanceSequence
  (Math|Integer|IncrementInt (Variables|Default|GetCurrentIdx)
    (:%20%20
      (Concatenator|PlayCurrentItem))))

(fn GetCurrentItem
  (return (Utilities|Array|Get(acopy) (Variables|Sequence|GetSoundSequence) (Variables|Default|GetCurrentIdx))))

(fn Play
  (params (AttachToComponent :type Scene%20Component%20Object%20Reference))
  (Networking|CanExecuteCosmeticEvents
    (:True
      (Variables|Default|SetAttachComponent AttachToComponent)
      (Concatenator|PlayCurrentItem))))

(fn PlayCurrentItem
  (bind _output_sound_2_d8314be3444a737f8a2ed3b3edeac6b5 (Utilities|Struct|BreakSoundConcatenatorSequenceItemData (Concatenator|GetCurrentItem)))
  (bind _outputdelegate (EventDispatchers|CreateEvent :!event StartDelay))
  (Utilities|IsValid _output_sound_2_d8314be3444a737f8a2ed3b3edeac6b5
    (:Is%20Valid
      (bind _attachcomponent (Variables|Default|GetAttachComponent)
        (:then
          (bind _returnvalue (Audio|SpawnSoundAttached _output_sound_2_d8314be3444a737f8a2ed3b3edeac6b5 _attachcomponent))
          (EventDispatchers|BindEventtoOnAudioFinished _returnvalue _outputdelegate))
        (:else
          (bind _returnvalue_2 (Audio|SpawnSound2D _output_sound_2_d8314be3444a737f8a2ed3b3edeac6b5))
          (EventDispatchers|BindEventtoOnAudioFinished _returnvalue_2 _outputdelegate))))))

(fn StartDelay
  (bind _output (Concatenator|GetCurrentItem))
  (bind (_sound_2_d8314be3444a737f8a2ed3b3edeac6b5 _delay_5_9b967cfa4e4664110f1201a1b0291d30) (Utilities|Struct|BreakSoundConcatenatorSequenceItemData _output))
  (if (== _delay_5_9b967cfa4e4664110f1201a1b0291d30 0.0)
    (Concatenator|AdvanceSequence)
    (else
      (Utilities|Time|SetTimerbyEvent (EventDispatchers|CreateEvent :!event AdvanceSequence) _delay_5_9b967cfa4e4664110f1201a1b0291d30))))
